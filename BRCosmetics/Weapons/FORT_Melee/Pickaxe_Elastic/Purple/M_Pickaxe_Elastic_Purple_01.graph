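# Unlit, translucent, two-sided effect material for the purple Elastic pickaxe, showing glowing animated noise with mesh distortion.
# Panned, rotatable noise drives the emissive color, blending secondary and primary colors through a V-based curve atlas row, with TwoSidedSign selecting a dimmer secondary-colored back side.
# Noise alpha times an inverted-V gradient with adjustable fade forms the opacity.
# A separately panned texture, masked to peak at mid-V, is transformed for world position offset.
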Material /BRCosmetics/Weapons/FORT_Melee/Pickaxe_Elastic/Purple/M_Pickaxe_Elastic_Purple_01
graph {
 properties {
  bUsedWithMeshDeformer=true
  bUsedWithSkeletalMesh=true
  blendMode=BLEND_Translucent
  pixelDepthOffsetMode=PDOM_Legacy
  shadingModel=MSM_Unlit
  twoSided=true
  usedShadingModels=Unlit
 }
 nodes [
  n1 = VectorParameter() {group=Color, parameterName=Secondary Color, defaultValue=<rgba 0.5 0 1 1>}
  n2 = Multiply(
   A=TextureCoordinate()
   B=AppendVector(
    A=ScalarParameter() {defaultValue=1.5, group=Noise, parameterName=NoiseTexture_Tile_U}
    B=ScalarParameter() {defaultValue=0.5, group=Noise, parameterName=NoiseTexture_Tile_V}
   )
  )
  n3 = AppendVector(A=ScalarParameter() {group=Noise, parameterName=NoiseTexture_Pan_U}, B=ScalarParameter() {group=Noise, parameterName=NoiseTexture_Pan_V})
  n4 = TextureSample(UVs=Rotator(Coordinate=Panner(Coordinate=@n2, Speed=@n3) {speedY=0.3}, Time=ScalarParameter() {group=Noise, parameterName=NoiseTexture_Rotator})) {texture=&/T_Pickaxe_Purple_Noise_01}
  n5 = Power(Base=@n4.RGB, Exp=ScalarParameter() {defaultValue=1, group=Color, parameterName=Texture_Power})
  n6 = ScalarParameter() {defaultValue=20, group=Color, parameterName=Color_Intensity}
  n7 = Multiply(A=Multiply(A=Multiply(A=@n1.RGB, B=@n5), B=@n6), B=ScalarParameter() {defaultValue=0.1, group=Color, parameterName=BackSide_Intensity}) {constA=0.1, constB=0}
  n8 = Clamp(None=TwoSidedSign())
  n9 = CurveAtlasRowParameter(CurveTime=ComponentMask(None=TextureCoordinate()) {g=true}) {
   atlas=/Game/Athena/LinearColorCurves/TempAthenaCurveAtlas
   curve=/Game/Weapons/FORT_Melee/Pickaxe_Elastic/Purple/Curve_Pickaxe_Elastic_Purple_02
   defaultValue=16
  }
  n10 = LinearInterpolate(A=@n1.RGB, B=VectorParameter() {group=Color, parameterName=Primary Color, defaultValue=<rgba 0.5 0 1 1>}.RGB, Alpha=@n9.R)
  n11 = ScalarParameter() {group=Color, parameterName=Color_Mix}
  n12 = Multiply(A=Multiply(A=LinearInterpolate(A=@n1.RGB, B=@n10, Alpha=@n11), B=@n5), B=@n6)
  MaterialFunctionCall(Emissive=Desaturation(None=LinearInterpolate(A=@n7, B=@n12, Alpha=@n8), Fraction=ScalarParameter() {group=Color, parameterName=Color Desaturation})) {materialFunction=&MF_FX_EmissiveAdjustments}
  n13 = OneMinus(None=ComponentMask(None=TextureCoordinate()) {g=true})
  n14 = ScalarParameter() {parameterName=Alpha_Fade}
  Multiply(A=@n4.A, B=Clamp(None=Add(A=@n13, B=@n14)))
  n15 = ScalarParameter() {defaultValue=1, group=WPO, parameterName=WPO_Intensity}
  n16 = TextureCoordinate()
  n17 = AppendVector(
   A=ScalarParameter() {defaultValue=2, group=WPO, parameterName=WPOTexture_Tile_U}
   B=ScalarParameter() {defaultValue=0.5, group=WPO, parameterName=WPOTexture_Tile_V}
  )
  n18 = ScalarParameter() {group=WPO, parameterName=WPOTexture_Pan_U}
  n19 = ScalarParameter() {defaultValue=-1, group=WPO, parameterName=WPOTexture_Pan_V}
  n20 = TextureSampleParameter2D(UVs=Panner(Coordinate=Multiply(A=@n16, B=@n17), Speed=AppendVector(A=@n18, B=@n19)) {speedY=0.3}) {group=WPO, parameterName=WPO_Texture, texture=&2ndInfectedMapLinear}
  n21 = Subtract(A=ComponentMask(None=TextureCoordinate()) {g=true}) {constB=0.5}
  n22 = Subtract(A=OneMinus(None=Abs(None=@n21))) {constB=0.5}
  Transform(None=Multiply(A=Multiply(A=@n15, B=@n20.RGB), B=Clamp(None=@n22)))
 ]
}
