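# Translucent lit material for the purple Elastic pickaxe, apparently an animated energy blade effect.
# A UV2-driven two-color blend is shaped by one curve atlas gradient sampled twice, along tiled, panned, rotated UVs and across the width, adding highlights.
# A sloped UV cutoff mask and the width profile form a further mask, alongside a view-angle fade.
# Panning noise drives world position offset along local-axis-masked vertex normals, attenuated along UV2.
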Material /BRCosmetics/Weapons/FORT_Melee/Pickaxe_Elastic/Purple/M_Pickaxe_Elastic_Purple_02
graph {
 properties {
  bUsedWithInstancedStaticMeshes=true
  bUsedWithMeshDeformer=true
  bUsedWithSkeletalMesh=true
  blendMode=BLEND_Translucent
  pixelDepthOffsetMode=PDOM_Legacy
  usedShadingModels=Default Lit
 }
 nodes [
  n1 = VectorParameter() {group=Color, parameterName=Secondary Color, defaultValue=<rgba 0.5 0 1 1>}
  n2 = ScalarParameter() {defaultValue=1, group=Color, parameterName=Secondary Color Intensity}
  n3 = VectorParameter() {group=Color, parameterName=Primary Color, defaultValue=<rgba 0.5 0 1 1>}
  n4 = ScalarParameter() {defaultValue=1, group=Color, parameterName=Primary Color Intensity}
  n5 = TextureCoordinate() {coordinateIndex=2}
  n6 = Power(
   Base=Add(A=ComponentMask(None=@n5) {r=true}, B=ScalarParameter() {defaultValue=-0.35, group=Color, parameterName=Add})
   Exp=ScalarParameter() {defaultValue=1.5, group=Color, parameterName=Power}
  )
  n7 = ScalarParameter() {defaultValue=6, group=Color, parameterName=Mult}
  n8 = LinearInterpolate(
   A=Desaturation(None=Multiply(A=@n1.RGB, B=@n2), Fraction=ScalarParameter() {group=Color, parameterName=Secondary Color Desaturation})
   B=Desaturation(None=Multiply(A=@n3.RGB, B=@n4), Fraction=ScalarParameter() {group=Color, parameterName=Primary Color Desaturation})
   Alpha=Clamp(None=Multiply(A=@n6, B=@n7) {constB=1.5}) {maxDefault=4, minDefault=-2}
  )
  MaterialFunctionCall(Emissive=@n8) {materialFunction=&MF_FX_EmissiveAdjustments}
  n9 = TextureCoordinate()
  n10 = AppendVector(
   A=ScalarParameter() {defaultValue=1, group=Gradient, parameterName=Gradient_Tile_U}
   B=ScalarParameter() {defaultValue=3, group=Gradient, parameterName=Gradient_Tile_V}
  )
  n11 = ScalarParameter() {group=Gradient, parameterName=Gradient_Pan_U}
  n12 = ScalarParameter() {group=Gradient, parameterName=Gradient_Pan_V}
  n13 = Rotator(
   Coordinate=Panner(Coordinate=Multiply(A=@n9, B=@n10), Speed=AppendVector(A=@n11, B=@n12)) {speedY=0.3}
   Time=ScalarParameter() {defaultValue=10, group=Gradient, parameterName=Gradient_Rotator}
  )
  n14 = CurveAtlasRowParameter(CurveTime=Frac(None=ComponentMask(None=@n13) {g=true})) {
   atlas=/Game/Athena/LinearColorCurves/TempAthenaCurveAtlas
   curve=/Game/Weapons/FORT_Melee/Pickaxe_Elastic/Purple/Curve_Pickaxe_Elastic_Purple_01
   defaultValue=13
  }
  n15 = ComponentMask(None=TextureCoordinate()) {r=true}
  n16 = OneMinus(None=Abs(None=Subtract(A=@n15) {constB=0.5}))
  n17 = CurveAtlasRowParameter(CurveTime=Clamp(None=Subtract(A=@n16) {constB=0.5})) {
   atlas=&TempAthenaCurveAtlas
   curve=&Curve_Pickaxe_Elastic_Purple_01
   defaultValue=13
  }
  n18 = Subtract(A=@n17.B, B=ScalarParameter() {defaultValue=0.1, group=Gradient, parameterName=Gradient_Blade_Fade})
  n19 = Multiply(A=@n17.B, B=OneMinus(None=@n18))
  n20 = Add(A=Subtract(A=Multiply(A=@n14.R, B=@n18), B=ScalarParameter() {defaultValue=0.1, group=Gradient, parameterName=Gradient_Center_Fade}), B=@n19)
  Multiply(
   A=LinearInterpolate(
    A=Multiply(A=@n8, B=@n14.A)
    B=ScalarParameter() {defaultValue=1, group=Gradient, parameterName=Gradient_HighLight_Intensity}
    Alpha=Clamp(None=@n20)
   )
   B=Multiply(A=@n14.G, B=@n17.B)
  )
  n21 = ScalarParameter() {parameterName=Param_2}
  n22 = TextureCoordinate()
  n23 = ComponentMask(None=@n22) {g=true}
  n24 = Subtract(A=ComponentMask(None=@n22) {r=true}) {constB=0.5}
  n25 = Multiply(A=ScalarParameter() {defaultValue=0.3, group=Mask, parameterName=Mask_Slope, sortPriority=2}, B=Multiply(A=Abs(None=@n24)) {constB=2})
  n26 = Add(A=Subtract(A=OneMinus(None=@n23), B=OneMinus(None=@n25)), B=ScalarParameter() {defaultValue=1.7, group=Mask, parameterName=Mask_Cut})
  Multiply(
   A=Multiply(A=Power(Base=@n19, Exp=@n21), B=ScalarParameter() {defaultValue=3, parameterName=Param_1})
   B=Power(Base=Clamp(None=@n26), Exp=ScalarParameter() {defaultValue=10, group=Mask, parameterName=Mask_Power})
  )
  Multiply(B=MaterialFunctionCall("FadeSensitivity (optional)"=ScalarParameter() {parameterName=ViewAngleFade}) {materialFunction=&FadeBasedOnViewAngle}.Result)
  n27 = Transform(None=VertexNormalWS()) {transformSourceType=TRANSFORMSOURCE_World, transformType=TRANSFORM_Local}
  n28 = Constant3Vector() {constant=<rgba 0 1 0 1>}
  n29 = Multiply(
   A=Transform(None=Multiply(A=@n27, B=@n28)) {transformSourceType=TRANSFORMSOURCE_Local}
   B=ScalarParameter() {group=WPO, parameterName=WPO_Intensity_Width}
  )
  n30 = ScalarParameter() {group=WPO, parameterName=WPO_Intensity_In, sortPriority=4}
  n31 = ScalarParameter() {defaultValue=1, group=WPO, parameterName=WPO_Intensity_Out, sortPriority=4}
  n32 = TextureCoordinate()
  n33 = AppendVector(
   A=ScalarParameter() {defaultValue=1, group=WPO, parameterName=WPO_Tile_U, sortPriority=0}
   B=ScalarParameter() {defaultValue=1, group=WPO, parameterName=WPO_Tile_V, sortPriority=0}
  )
  n34 = ScalarParameter() {group=WPO, parameterName=WPO_Pan_U, sortPriority=2}
  n35 = ScalarParameter() {group=WPO, parameterName=WPO_Pan_V, sortPriority=2}
  n36 = TextureSampleParameter2D(UVs=Panner(Coordinate=Multiply(A=@n32, B=@n33), Speed=AppendVector(A=@n34, B=@n35)) {speedY=0.3}) {
   group=WPO
   parameterName=WPO_Texture
   samplerType=SAMPLERTYPE_LinearGrayscale
   sortPriority=0
   texture=/Engine/Functions/Engine_MaterialFunctions02/ExampleContent/Textures/LowResBlurredNoise
  }
  n37 = ComponentMask(None=TextureCoordinate() {coordinateIndex=2}) {g=true}
  n38 = LinearInterpolate(A=OneMinus(None=@n37), B=@n37, Alpha=ScalarParameter() {parameterName=Flip WPO UV's})
  n39 = ScalarParameter() {group=WPO, parameterName=WPO_attenuation}
  n40 = Multiply(
   A=Power(Base=LinearInterpolate(A=@n30, B=@n31, Alpha=@n36.R) {constA=-1}, Exp=ScalarParameter() {defaultValue=1, group=WPO, parameterName=WPO_Power})
   B=Clamp(None=Subtract(A=@n38, B=@n39))
  )
  n41 = Multiply(A=@n19, B=@n40)
  n42 = Transform(None=VertexNormalWS()) {transformSourceType=TRANSFORMSOURCE_World, transformType=TRANSFORM_Local}
  n43 = Constant3Vector() {constant=<rgba 0 0 1 1>}
  n44 = Multiply(
   A=Transform(None=Multiply(A=@n42, B=@n43)) {transformSourceType=TRANSFORMSOURCE_Local}
   B=ScalarParameter() {group=WPO, parameterName=WPO_Intensity}
  )
  Reroute(None=Add(A=Multiply(A=@n29, B=@n41), B=Multiply(A=@n44, B=@n40)))
 ]
}
