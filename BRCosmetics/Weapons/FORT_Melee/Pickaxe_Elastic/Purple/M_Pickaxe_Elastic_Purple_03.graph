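# Unlit, two-sided masked material for the purple Elastic pickaxe, producing a glowing, animated cutout effect.
# A panning mask texture, distorted by scrolling noise, is shaped by intensity and power, then multiplied by UV V.
# An adjustable V gradient is added, and an If node thresholds the sum into a hard binary mask.
# Emissive multiplies Primary Color and intensity by a curve atlas channel sampled along V, then passes through MF_FX_EmissiveAdjustments.
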Material /BRCosmetics/Weapons/FORT_Melee/Pickaxe_Elastic/Purple/M_Pickaxe_Elastic_Purple_03
graph {
 properties {
  bUsedWithMeshDeformer=true
  bUsedWithSkeletalMesh=true
  blendMode=BLEND_Masked
  pixelDepthOffsetMode=PDOM_Legacy
  shadingModel=MSM_Unlit
  twoSided=true
  usedShadingModels=Unlit
 }
 nodes [
  n1 = Multiply(
   A=TextureCoordinate()
   B=AppendVector(
    A=ScalarParameter() {defaultValue=0.8, group=Mask, parameterName=MaskTexture_Tile_U}
    B=ScalarParameter() {defaultValue=1, group=Mask, parameterName=MaskTexture_Tile_V}
   )
  )
  n2 = AppendVector(
   A=ScalarParameter() {defaultValue=1, group=Mask, parameterName=MaskTexture_Pan_U}
   B=ScalarParameter() {group=Mask, parameterName=MaskTexture_Pan_V}
  )
  n3 = TextureCoordinate()
  n4 = AppendVector(
   A=ScalarParameter() {defaultValue=2, group=Distortion, parameterName=Distortion_Tile_U}
   B=ScalarParameter() {defaultValue=0.5, group=Distortion, parameterName=Distortion_Tile_V}
  )
  n5 = ScalarParameter() {defaultValue=-0.5, group=Distortion, parameterName=Distortion_Pan_U}
  n6 = ScalarParameter() {defaultValue=0.3, group=Distortion, parameterName=Distortion_Pan_V}
  n7 = TextureSample(UVs=Panner(Coordinate=Multiply(A=@n3, B=@n4), Speed=AppendVector(A=@n5, B=@n6)) {speedY=0.3}) {texture=&T_NoisePack_01}
  n8 = ScalarParameter() {defaultValue=0.1, group=Distortion, parameterName=Distortion_Intensity}
  n9 = TextureSampleParameter2D(UVs=Add(A=Panner(Coordinate=@n1, Speed=@n2) {speedY=0.3}, B=Multiply(A=@n7.R, B=@n8) {constB=0.1})) {group=Mask, parameterName=MaskTexture, texture=/Game/Characters/Player/Male/Small/T3_Items/Shuriken/Textures/T_NinjaStarDistort}
  n10 = Multiply(A=ScalarParameter() {defaultValue=2, group=Mask, parameterName=Mask_Intensity}, B=@n9.R)
  n11 = ScalarParameter() {defaultValue=2, group=Mask, parameterName=Mask_Power}
  n12 = ComponentMask(None=TextureCoordinate()) {g=true}
  n13 = Power(Base=@n12, Exp=ScalarParameter() {defaultValue=2, group=Mask, parameterName=Mask_Hardness})
  n14 = Add(
   A=Multiply(A=Power(Base=@n10, Exp=@n11), B=@n12)
   B=Multiply(A=Clamp(None=@n13), B=ScalarParameter() {defaultValue=1, group=Mask, parameterName=Mask_Cut})
  )
  n15 = Constant() {r=1}
  If(A=@n14, B=Constant() {r=0.3}, "A > B"=@n15, "A == B"=@n15, "A < B"=Constant())
  n16 = VectorParameter() {group=Color, parameterName=Primary Color, defaultValue=<rgba 0.5 0 1 1>}
  n17 = ScalarParameter() {defaultValue=1, group=color, parameterName=Color_Intensity}
  n18 = CurveAtlasRowParameter(CurveTime=ComponentMask(None=TextureCoordinate()) {g=true}) {
   atlas=/Game/Athena/LinearColorCurves/TempAthenaCurveAtlas
   curve=/Game/Weapons/FORT_Melee/Pickaxe_Elastic/Purple/Curve_Pickaxe_Elastic_Purple_02
   defaultValue=16
  }
  MaterialFunctionCall(Emissive=Multiply(A=Multiply(A=@n16.RGB, B=@n17), B=@n18.G)) {materialFunction=&MF_FX_EmissiveAdjustments}
  Multiply(A=@n14, B=ScalarParameter() {defaultValue=1, parameterName=Alpha})
  VectorParameter() {group=Color, parameterName=Secondary Color, defaultValue=<rgba 0.5 0 1 1>}
 ]
}
