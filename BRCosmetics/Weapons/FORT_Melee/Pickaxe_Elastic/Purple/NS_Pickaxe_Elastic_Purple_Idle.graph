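# Niagara idle effect for the purple Elastic pickaxe, layering smoke, plasma, glows and sparks.
# Secondary Color tints the drifting smoke, plasma mesh, dim glow sprite and bright glow with attached light, whose sprite size flickers on system loop age.
# Primary Color tints three spark emitters: a noisy directional stream, a vortex-swirled shell, and falling sparks spawned per unit moved, so they appear only when the effect moves.
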NiagaraSystem /BRCosmetics/Weapons/FORT_Melee/Pickaxe_Elastic/Purple/NS_Pickaxe_Elastic_Purple_Idle
dependencies {
 usedDataInterfaces [&NiagaraDataInterfaceCurve, &NiagaraDataInterfaceVector2DCurve]
 usedDynamicInputs []
 usedModules [&SystemState]
}
emitters {
 Decal_Smoke {
  emitterUpdateScript [
   &EmitterState()
   &SpawnBurst_Instantaneous(Spawn Count=1, Use Spawn Probability=false) {enabled=false}
   &SpawnRate(SpawnRate=&UniformRangedFloat(Minimum=3, Maximum=5), Use Spawn Probability=false)
  ]
  particleSpawnScript [
   /Game/Effects/Niagara/Modules/Partition/FN_ParticlePartition_v2/FN_ParticlePartition_v2()
   /Niagara/Modules/Spawn/Initialization/InitializeParticle(
    Color=$User.Secondary Color
    Lifetime=&UniformRangedFloat(Minimum=0.9, Maximum=1)
    Module.Write Color=true
    Module.Write Lifetime=true
    Module.Write Mass=true
    Module.Write Material Random=true
    Module.Write Position=false
    Module.Write Scale=false
    Module.Write SpriteRotation=true
    Module.Write SpriteSize=true
    Position=&SimulationPosition()
    Sprite Rotation=&UniformRangedFloat(Minimum=90, Maximum=-90)
    Sprite Size=&Vector2DFromFloat(Value=&UniformRangedFloat(Minimum=20, Maximum=25))
   )
   &SystemLocation(MaskSpawnGroup=false, Offset=&UniformRangedVector(Minimum=<12 0 0>, Maximum=<6 0 0>))
   &AddVelocity(Velocity=&UniformRangedVector(Minimum=<-50 0 0>, Maximum=<-60 0 0>))
  ]
  particleUpdateScript [
   &ParticleState()
   &ScaleColor(
    Scale Alpha=&FloatFromCurve(
     FloatCurve=DataInterface() {propertyValues {
       bOptimizeLUT=false
       Curve {keys [
         {interpMode=RCIM_Cubic, tangentMode=RCTM_Break, time=0, value=0}
         {arriveTangent=1.7205, interpMode=RCIM_Cubic, leaveTangent=1.7205, tangentMode=RCTM_User, time=0.106, value=1.3774}
         {arriveTangent=-1.5407, interpMode=RCIM_Cubic, leaveTangent=-1.5407, time=0.5942, value=1.3975}
         {interpMode=RCIM_Cubic, tangentMode=RCTM_Break, time=1, value=0}
        ]}
      }}
     Scale Curve=0.5
    )
    Scale RGB=&VectorFromFloat(Value=&FloatFromCurve(FloatCurve=DataInterface() {propertyValues {Curve {keys [
          {arriveTangent=8.3168, interpMode=RCIM_Cubic, leaveTangent=8.3168, tangentMode=RCTM_User, time=0.0167, value=185.0701}
          {arriveTangent=-162.7845, interpMode=RCIM_Cubic, leaveTangent=-162.7845, time=0.3189, value=140.0851}
          <at 1 25>
         ]}}}))
    ScaleA=true
    ScaleRGB=true
   )
   &ScaleSpriteSize(Scale Factor=&Vector2DFromFloat(Value=&FloatFromCurve(FloatCurve=DataInterface() {propertyValues {Curve {keys [
          {interpMode=RCIM_Cubic, tangentMode=RCTM_Break, time=0.008, value=1.0739}
          {arriveTangent=-2.4972, interpMode=RCIM_Cubic, tangentMode=RCTM_Break, time=1, value=0}
         ]}}})))
   &CameraOffset(Camera Offset Amount=12)
   &SolveForcesAndVelocity(Clamp Velocity=false, Limit Acceleration=false)
   &DynamicMaterialParameters(
    Index 0 Param 1=&FloatFromCurve(FloatCurve=DataInterface() {propertyValues {Curve {keys [{interpMode=RCIM_Cubic, leaveTangent=2.3044, tangentMode=RCTM_Break, time=0, value=0.25}, <at 1 1>]}}})
    Index 0 Param 2=0.5
    Module.Param0WriteEnabled=true
    Module.Param1WriteEnabled=true
    Module.Param2WriteEnabled=true
    Module.Param3WriteEnabled=true
   )
  ]
  rendererClasses [&NiagaraSpriteRendererProperties]
  renderers [{rendererClass=&NiagaraSpriteRendererProperties}]
 }
 Energy_Plasma {
  emitterUpdateScript [
   &EmitterScalability(
    Module.bApplySpawnCountScale=false
    Module.bApplySpawnCountScaleByDistanceFraction=false
    Module.bUseMaxDistance=false
    Module.bUseMinDistance=false
   ) {enabled=false}
   &EmitterLifeCycle(NextLoopDuration=4.4208) {enabled=false}
   &EmitterState()
   &SpawnBurst_Instantaneous(Spawn Count=1, Spawn Time=0.07, Use Spawn Probability=false) {enabled=false}
   &SpawnRate(SpawnRate=3, Use Spawn Probability=false)
  ]
  particleSpawnScript [
   &InitializeParticle(
    Color=$User.Secondary Color
    Mesh Scale=&Multiply_VectorByFloat(
     Vector=&VectorFromFloat(Value=0.72)
     Float=&UniformRangedFloat(Minimum=0.22, Maximum=0.2)
    )
    Module.Write Color=true
    Module.Write Lifetime=true
    Module.Write Mass=true
    Module.Write Material Random=true
    Module.Write Position=false
    Module.Write Scale=true
    Module.Write SpriteRotation=false
    Module.Write SpriteSize=false
    Position=&SimulationPosition()
   )
   /Niagara/Modules/Spawn/Orientation/V2/InitialMeshOrientation(
    Orientation Axis:Vector3f=<0 0 1>
    Orientation Vector=<0 1 0>
    Rotation=&MakeVector(X=&UniformRangedFloat(Minimum=0, Maximum=1), Y=0, Z=0)
    Use Orientation Vector=true
    Use Rotation Vector=true
   )
  ]
  particleUpdateScript [
   &CameraOffset(Camera Offset Amount=25)
   &ParticleState()
   &ScaleColor(
    Scale Alpha=&FloatFromCurve(FloatCurve=DataInterface() {propertyValues {Curve {keys [
         {interpMode=RCIM_Cubic, tangentMode=RCTM_User, time=0.0023, value=0}
         {arriveTangent=0.9299, interpMode=RCIM_Cubic, leaveTangent=-0.254, tangentMode=RCTM_Break, time=0.3489, value=1}
         {arriveTangent=-0.1758, interpMode=RCIM_Cubic, leaveTangent=-0.1758, tangentMode=RCTM_User, time=1.0011, value=0.002}
        ]}}})
    Scale RGB=&VectorFromFloat(Value=&FloatFromCurve(
      FloatCurve=DataInterface() {propertyValues {Curve {keys [
          {interpMode=RCIM_Cubic, tangentMode=RCTM_Break, time=0, value=2}
          {arriveTangent=-2.9307, interpMode=RCIM_Cubic, tangentMode=RCTM_Break, time=1, value=0}
         ]}}}
      Scale Curve=20
     ))
    ScaleA=true
    ScaleRGB=true
   )
   &AddVelocity(Velocity=&UniformRangedVector(Minimum=<-6 0 0>, Maximum=<3 0 0>))
   &SolveForcesAndVelocity(Clamp Velocity=false, Limit Acceleration=false)
   &DynamicMaterialParameters(
    Index 0 Param 1=&FloatFromCurve(FloatCurve=DataInterface() {propertyValues {Curve {keys [<at 0 2.5>,<at 1 0>]}}})
    Index 0 Param 2=&FloatFromCurve(FloatCurve=DataInterface() {propertyValues {Curve {keys [<at 0 3>,<at 1 0>]}}}, Scale Curve=5)
    Index 0 Param 3=1
    Module.Param0WriteEnabled=true
    Module.Param1WriteEnabled=true
    Module.Param2WriteEnabled=true
    Module.Param3WriteEnabled=true
   )
   &CameraOffset(Camera Offset Amount=20) {module=CameraOffset001}
  ]
  rendererClasses [&NiagaraMeshRendererProperties]
  renderers [{rendererClass=&NiagaraMeshRendererProperties, rendererIndex=0}]
 }
 GlowDark {
  emitterUpdateScript [
   &EmitterState(Inactive Response=::NewEnumerator1, Life Cycle Mode=::NewEnumerator1, Loop Duration=0.3, UseLoopDelay=false)
   &SpawnBurst_Instantaneous(Spawn Count=1, Use Spawn Probability=false)
  ]
  particleSpawnScript [
   &InitializeParticle(
    Color=$User.Secondary Color
    Lifetime=0.3
    Module.Write Color=true
    Module.Write Lifetime=true
    Module.Write Mass=false
    Module.Write Material Random=true
    Module.Write Position=false
    Module.Write Scale=false
    Module.Write SpriteRotation=false
    Module.Write SpriteSize=true
    Position=&SimulationPosition()
    Sprite Size=&Vector2DFromFloat(Value=30)
   )
   &DynamicMaterialParameters(
    Index 0 Param 1=1
    Module.Param0WriteEnabled=true
    Module.Param1WriteEnabled=true
    Module.Param2WriteEnabled=true
    Module.Param3WriteEnabled=true
   )
   &SystemLocation(MaskSpawnGroup=false, Offset=<2 -2 -2>)
  ]
  particleUpdateScript [
   &ParticleState()
   &SolveForcesAndVelocity(Clamp Velocity=false, Limit Acceleration=false)
   &ScaleColor(Scale RGB=&VectorFromFloat(Value=0.1), ScaleA=true, ScaleRGB=true)
   &ScaleSpriteSize(Scale Factor=&Vector2DFromFloat(Value=&FloatFromCurve(
      FloatCurve=DataInterface() {propertyValues {Curve {keys [<at 0 1>,<at 0.1341 0.726>,<at 0.2151 1.053>,<at 0.4084 0.5013>,<at 0.684 1.1084>,<at 0.8681 0.6573>,<at 0.9922 1>]}}}
      CurveIndex=$System.NormalizedLoopAge
     ))) {enabled=false}
  ]
  rendererClasses [&NiagaraSpriteRendererProperties]
  renderers [{rendererClass=&NiagaraSpriteRendererProperties}]
 }
 Light {
  emitterUpdateScript [
   &EmitterState(Inactive Response=::NewEnumerator1, Life Cycle Mode=::NewEnumerator1, Loop Duration=0.3, UseLoopDelay=false)
   &SpawnBurst_Instantaneous(Spawn Count=1, Use Spawn Probability=false)
  ]
  particleSpawnScript [
   &InitializeParticle(
    Color=$User.Secondary Color
    Lifetime=0.3
    Module.Write Color=true
    Module.Write Lifetime=true
    Module.Write Mass=false
    Module.Write Material Random=true
    Module.Write Position=false
    Module.Write Scale=false
    Module.Write SpriteRotation=false
    Module.Write SpriteSize=true
    Position=&SimulationPosition()
    Sprite Size=&Vector2DFromFloat(Value=80)
   )
   &DynamicMaterialParameters(
    Index 0 Param 1=1
    Module.Param0WriteEnabled=true
    Module.Param1WriteEnabled=true
    Module.Param2WriteEnabled=true
    Module.Param3WriteEnabled=true
   )
   &SystemLocation(MaskSpawnGroup=false, Offset=<2 -2 -2>)
  ]
  particleUpdateScript [
   &CameraOffset(Camera Offset Amount=15)
   &ParticleState()
   &SolveForcesAndVelocity(Clamp Velocity=false, Limit Acceleration=false)
   &ScaleColor(Scale RGB=&VectorFromFloat(Value=1.2), ScaleA=true, ScaleRGB=true)
   &FN_Light_v2(Light Exponent=4, Light Radius=50, LightEnabled=true, Position Offset:Vector3f=<5 -5 -5>, Scale Color=20)
   &ScaleSpriteSize(Scale Factor=&Vector2DFromFloat(Value=&FloatFromCurve(
      FloatCurve=DataInterface() {propertyValues {Curve {keys [<at 0 1>,<at 0.1341 0.726>,<at 0.2151 1.053>,<at 0.4084 0.5013>,<at 0.684 1.1084>,<at 0.8681 0.6573>,<at 0.9922 1>]}}}
      CurveIndex=$System.NormalizedLoopAge
     )))
  ]
  rendererClasses [&NiagaraSpriteRendererProperties, &NiagaraLightRendererProperties]
  renderers [
   {rendererClass=&NiagaraSpriteRendererProperties}
   {rendererClass=&NiagaraLightRendererProperties, rendererIndex=1}
  ]
 }
 Particle_Spark {
  emitterUpdateScript [
   &EmitterState()
   &SpawnBurst_Instantaneous(Spawn Count=1, Use Spawn Probability=false) {enabled=false}
   &SpawnRate(SpawnRate=&UniformRangedFloat(Minimum=16, Maximum=32), Use Spawn Probability=false)
  ]
  particleSpawnScript [
   &FN_ParticlePartition_v2()
   &InitializeParticle(
    Color=$User.Primary Color
    Lifetime=&UniformRangedFloat(Minimum=0.5, Maximum=0.7)
    Module.Write Color=true
    Module.Write Lifetime=true
    Module.Write Mass=true
    Module.Write Material Random=true
    Module.Write Position=false
    Module.Write Scale=false
    Module.Write SpriteRotation=true
    Module.Write SpriteSize=true
    Position=&SimulationPosition()
    Sprite Rotation=&UniformRangedFloat(Minimum=180, Maximum=-180)
    Sprite Size=&Multiply_Vector2(
     A=<1 2.5>
     B=&Vector2DFromFloat(Value=&UniformRangedFloat(Minimum=1, Maximum=1.5))
    )
   )
   /Game/Effects/Niagara/Modules/BoneFlow/FN_Initialize_BoneFlow_Attributes_v2()
   &SystemLocation(MaskSpawnGroup=false, Offset=<-10 0 0>)
   &AddVelocity(Velocity=&UniformRangedVector(Minimum=<-200 0 0>, Maximum=<-50 0 0>))
   &BoxLocation(Box Offset=<10 0 0>, Box Size=<20 5 5>, MaskSpawnGroup=false, Module.Override Seed=false, Rotate Box=false, Surface Only=false)
  ]
  particleUpdateScript [
   &ParticleState()
   &ScaleColor(
    Scale Alpha=2
    Scale RGB=&VectorFromFloat(Value=&UniformRangedFloat(Minimum=5, Maximum=75))
    ScaleA=true
    ScaleRGB=true
   )
   &CameraOffset()
   &Drag(Drag=4, Use Linear Drag=true, Use Rotational Drag=false)
   &VectorNoiseForce(Force Amount=400, Module.Use Random Noise Time Interval=false)
   &SolveForcesAndVelocity(Clamp Velocity=false, Limit Acceleration=false)
   &ScaleSpriteSize(Scale Factor=&Vector2DFromCurve(
     Vector2Curve=DataInterface() {propertyValues {
       XCurve {keys [<at 0 1>,<at 1 0.5>]}
       YCurve {keys [
         {interpMode=RCIM_Cubic, tangentMode=RCTM_Break, time=0, value=1}
         {arriveTangent=-1.9076, interpMode=RCIM_Cubic, leaveTangent=-1.9076, tangentMode=RCTM_User, time=1, value=0}
        ]}
      }}
     ScaleCurve=<1 1>
    ))
  ]
  rendererClasses [&NiagaraSpriteRendererProperties]
  renderers [{rendererClass=&NiagaraSpriteRendererProperties}]
 }
 Particle_Spark001 {
  emitterUpdateScript [
   &EmitterState()
   &SpawnBurst_Instantaneous(Spawn Count=1, Use Spawn Probability=false) {enabled=false}
   &SpawnRate(SpawnRate=&UniformRangedFloat(Minimum=20, Maximum=12), Use Spawn Probability=false)
  ]
  particleSpawnScript [
   /Game/Effects/Niagara/Modules/Orbit/FN_Orbit_Init_Variables_v2()
   &FN_ParticlePartition_v2()
   &InitializeParticle(
    Color=$User.Primary Color
    Lifetime=&UniformRangedFloat(Minimum=0.5, Maximum=0.35)
    Module.Write Color=true
    Module.Write Lifetime=true
    Module.Write Mass=true
    Module.Write Material Random=true
    Module.Write Position=false
    Module.Write Scale=false
    Module.Write SpriteRotation=true
    Module.Write SpriteSize=true
    Position=&SimulationPosition()
    Sprite Rotation=&UniformRangedFloat(Minimum=180, Maximum=-180)
    Sprite Size=&Multiply_Vector2(
     A=<1.5 3>
     B=&Vector2DFromFloat(Value=&UniformRangedFloat(Minimum=0.7, Maximum=1))
    )
   )
   &FN_Initialize_BoneFlow_Attributes_v2()
   &SystemLocation(MaskSpawnGroup=false, Offset=<2.5 0 0>)
   &SphereLocation(
    MaskSpawnGroup=false
    Non Uniform Scale=<0.5 0.7 0.5>
    Override Seed=false
    Sphere Radius=15
    Surface Only=true
    Surface Only Band Thickness=1
    UseNonUniformScale=true
    UseOffset=false
    UseSphereOrientation=false
   )
   &AddVelocity(Velocity=&UniformRangedVector(Minimum=<-15 0 0>, Maximum=<-10 0 0>))
  ]
  particleUpdateScript [
   &ParticleState()
   &ScaleColor(
    Scale Alpha=2
    Scale RGB=&VectorFromFloat(Value=&UniformRangedFloat(Minimum=5, Maximum=75))
    ScaleA=true
    ScaleRGB=true
   )
   &CameraOffset(Camera Offset Amount=20)
   &Drag(Drag=2, Use Linear Drag=true, Use Rotational Drag=false)
   &VectorNoiseForce(Module.Use Random Noise Time Interval=false)
   /Niagara/Modules/Update/Velocity/VortexVelocity(
    Influence Falloff Radius=80
    Module.Use Influence Falloff=true
    Velocity Amount:NiagaraFloat=&UniformRangedFloat(Minimum=16, Maximum=32)
    Vortex Axis:Vector3f=<1 0 0>
   )
   &SolveForcesAndVelocity(Clamp Velocity=false, Limit Acceleration=false)
   &ScaleSpriteSize(Scale Factor=&Vector2DFromCurve(
     Vector2Curve=DataInterface() {propertyValues {
       XCurve {keys [<at 0 1>,<at 1 0.5>]}
       YCurve {keys [
         {interpMode=RCIM_Cubic, tangentMode=RCTM_Break, time=0, value=1}
         {arriveTangent=-1.9076, interpMode=RCIM_Cubic, leaveTangent=-1.9076, tangentMode=RCTM_User, time=1, value=0}
        ]}
      }}
     ScaleCurve=<1 1>
    ))
  ]
  rendererClasses [&NiagaraSpriteRendererProperties]
  renderers [{rendererClass=&NiagaraSpriteRendererProperties}]
 }
 Particle_Spark_Drop {
  emitterUpdateScript [
   &EmitterState()
   &SpawnBurst_Instantaneous(Spawn Count=1, Use Spawn Probability=false) {enabled=false}
   &SpawnRate(SpawnRate=&UniformRangedFloat(Minimum=8, Maximum=10), Use Spawn Probability=false) {enabled=false}
   &SpawnPerUnit(Module.Use Max Movement Threshold=true, Module.Use Movement Tolerance=false, Spawn Per Unit=75, Use Spawn Probability=false)
  ]
  particleSpawnScript [
   &FN_ParticlePartition_v2()
   &InitializeParticle(
    Color=$User.Primary Color
    Lifetime=&UniformRangedFloat(Minimum=0.5, Maximum=1)
    Module.Write Color=true
    Module.Write Lifetime=true
    Module.Write Mass=true
    Module.Write Material Random=true
    Module.Write Position=false
    Module.Write Scale=false
    Module.Write SpriteRotation=true
    Module.Write SpriteSize=true
    Position=&SimulationPosition()
    Sprite Rotation=&UniformRangedFloat(Minimum=180, Maximum=-180)
    Sprite Size=&Multiply_Vector2(
     A=<1 3>
     B=&Vector2DFromFloat(Value=&UniformRangedFloat(Minimum=1, Maximum=1.5))
    )
   )
   &FN_Initialize_BoneFlow_Attributes_v2()
   &SystemLocation(MaskSpawnGroup=false, Offset=<-50 0 0>)
   &BoxLocation(Box Size=<50 5 20>, MaskSpawnGroup=false, Module.Override Seed=false, Rotate Box=false, Surface Only=false)
   &AddVelocity(Velocity=&UniformRangedVector(Minimum=<-200 0 0>, Maximum=<-80 0 0>)) {enabled=false}
  ]
  particleUpdateScript [
   &ParticleState()
   &ScaleColor(
    Scale Alpha=2
    Scale RGB=&VectorFromFloat(Value=&UniformRangedFloat(Minimum=0.1, Maximum=3))
    ScaleA=true
    ScaleRGB=true
   )
   &CameraOffset(Camera Offset Amount=20)
   &InheritVelocity(Module.Limit Speed=true, Velocity Limit=1)
   &AccelerationForce(Acceleration=&UniformRangedVector(Minimum=<0 0 -50>, Maximum=<1 1 -10>))
   &Drag(Drag=2, Use Linear Drag=true, Use Rotational Drag=false)
   &VectorNoiseForce(Module.Use Random Noise Time Interval=false)
   &SolveForcesAndVelocity(Clamp Velocity=false, Limit Acceleration=false)
   &ScaleSpriteSize(Scale Factor=&Vector2DFromCurve(
     Vector2Curve=DataInterface() {propertyValues {
       XCurve {keys [<at 0 1>,<at 1 0.5>]}
       YCurve {keys [
         {interpMode=RCIM_Cubic, tangentMode=RCTM_Break, time=0, value=1}
         {arriveTangent=-1.9076, interpMode=RCIM_Cubic, leaveTangent=-1.9076, tangentMode=RCTM_User, time=1, value=0}
        ]}
      }}
     ScaleCurve=<1 1>
    ))
  ]
  rendererClasses [&NiagaraSpriteRendererProperties]
  renderers [{rendererClass=&NiagaraSpriteRendererProperties}]
 }
}
summary {
 systemName=NS_Pickaxe_Elastic_Purple_Idle
 emitters [
  {emitterName=Light, rendererClasses [&NiagaraSpriteRendererProperties, &NiagaraLightRendererProperties]}
  {emitterName=Particle_Spark, rendererClasses [&NiagaraSpriteRendererProperties]}
  {emitterName=Energy_Plasma, rendererClasses [&NiagaraMeshRendererProperties]}
  {emitterName=Particle_Spark_Drop, rendererClasses [&NiagaraSpriteRendererProperties]}
  {emitterName=Particle_Spark001, rendererClasses [&NiagaraSpriteRendererProperties]}
  {emitterName=GlowDark, rendererClasses [&NiagaraSpriteRendererProperties]}
  {emitterName=Decal_Smoke, rendererClasses [&NiagaraSpriteRendererProperties]}
 ]
 userVariables [
  User.Primary Color {type=LinearColor, description="", defaultValue {struct=LinearColor, value=<rgba 0.5 0 1 1>}}
  User.Secondary Color {type=LinearColor, description="", defaultValue {struct=LinearColor, value=<rgba 0.5 0 1 1>}}
 ]
}
