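# Physics asset for the purple one-handed Elastic pickaxe, a melee weapon cosmetic.
# A single capsule body on the Root bone approximates the pickaxe's shape, offset along the bone's X axis and rotated about ninety degrees.
# It has no constraints, so it defines one collision volume with no jointed or articulated parts.
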PhysicsAsset /BRCosmetics/Weapons/FORT_Melee/Pickaxe_Elastic/Purple/SM_Pickaxe_Elastic1H_Purple_PhysicsAsset
bodies {Root [{shapeType=Capsule, length=87.1712, radius=19.7696, center=<64.6196 -0.1771 -0.2408>, rotation=<rot 0 -89.8253 -89.2256>}]}
constraints []
